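# Niagara impact effect for the BombaToon bass pickaxe, layering surface debris with cartoon flashes, sparks and musical notes.
# FN_PhysMatDebris uses User.PhysicalSurface to drive debris velocity and gravity, with its collision module disabled.
# Flash_Elements switches layered sprite settings by spawn group; Notes bursts three times with curl noise and wobble; Toon_Sparks are randomly black or white, speed-stretched streaks.
# Front-end scaling resizes flashes, notes and sparks.
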NiagaraSystem /SparksCosmetics/Instrument/Bass/BombaToon/Effects/NS_Pickaxe_Bass_BombaToon_Impact
dependencies {
 usedDataInterfaces [
  &NiagaraDataInterfaceCurve
  &NiagaraDataInterfaceColorCurve
  &NiagaraDataInterfaceMeshRendererInfo
  &NiagaraDataInterfaceVector2DCurve
 ]
 usedDynamicInputs [&FNTimeOfDayScalar]
 usedModules [
  &self:SystemSpawnScript.SystemScriptSource.SystemScriptGraph.NiagaraNodeAssignment_0.SetVariables_#1
  &SystemState
 ]
}
emitters {
 FN_PhysMatDebris {
  emitterUpdateScript [
   &EmitterState()
   &SpawnBurst_Instantaneous(
    Spawn Count=&RandomRangeInt(Minimum=5, Maximum=10)
    Use Spawn Probability=false
    UseLoopCountLimit=false
   )
  ]
  particleSpawnScript [
   /SetVariables(Particles.BiasedRandom=&FloatFromCurve(
     FloatCurve=DataInterface() {propertyValues {Curve {keys [
         <at 0.2 1 RCIM_Cubic>
         {arriveTangent=-3.3394, interpMode=RCIM_Cubic, leaveTangent=-3.3394, tangentMode=RCTM_User, time=0.5, value=0.5}
         <at 0.8 0 RCIM_Cubic>
        ]}}}
     CurveIndex=&RandomRangeFloat(Minimum=0, Maximum=1)
    )) {isSetParameters=true}
   &InitializeParticle(
    Color Mode=::NewEnumerator0
    Lifetime=&RandomRangeFloat(Minimum=0.5, Maximum=1.25)
    Mass=&Lerp_Float(A=0.6, B=1.2, Alpha=$Particles.BiasedRandom)
    Mass Mode=::NewEnumerator1
    Mesh Scale=&Multiply_VectorByFloat(
     Vector=&RandomRangeVector(Minimum=<0.9 0.4 0.5>, Maximum=<1 1 1>)
     Float=&Lerp_Float(A=0.7, B=1, Alpha=$Particles.BiasedRandom)
    )
    Mesh Scale Mode=::NewEnumerator4
    UsePositionOffset=false
    Write Material Random=true
   )
   &FN_PhysicalSurface(
    Dirt Mesh Count=2
    Glass Mesh Count=3
    Ice Mesh Count=2
    Metal Gravity=&RandomRangeFloat(Minimum=-350, Maximum=-1800)
    Metal Mesh Count=2
    Metal Velocity=&RandomRangeFloat(Minimum=80, Maximum=375)
    Metal Velocity Cone Angle=130
    Physical Surface=$User.PhysicalSurface
    Stone Gravity=&RandomRangeFloat(Minimum=-700, Maximum=-1250)
    Stone Mesh Count=2
    Stone Velocity=&RandomRangeFloat(Minimum=80, Maximum=550)
    Stone Velocity Cone Angle=120
    Wood Gravity=&RandomRangeFloat(Minimum=-450, Maximum=-900)
    Wood Mesh Count=2
    Wood Scale=1.5
    Wood Velocity=&RandomRangeFloat(Minimum=80, Maximum=600)
    Wood Velocity Cone Angle=130
   )
   &AddVelocityInCone(
    Cone Angle=$Output.FN_PhysicalSurface.SurfaceResponseConeAngle
    Module.Use Velocity Falloff On Cone Axis=false
    Velocity Strength=$Output.FN_PhysicalSurface.SurfaceResponseVelocity
   )
   &InitialMeshOrientation(Use Rotation Vector=false)
   &AddRotationalVelocity(Rotation Rate=&RandomVector(Vector Scale=&RandomRangeFloat(Minimum=1, Maximum=2)))
   &ApplyInitialForces()
   /SetVariables(Particles.GravityOverrideValue=$Output.FN_PhysicalSurface.GravityOverride) {isSetParameters=true}
  ]
  particleUpdateScript [
   &ParticleState()
   &GravityForce(Gravity=$Particles.GravityOverrideValue)
   &Collision(
    Advanced Aging Rate=1
    Age Colliding Particles=false
    CPU Collision Trace Channel=ECC_WorldStatic
    Collision Enabled=&Boolean_AndOperation(
     Bool1=&SetBoolByFloatComparison(A=$Particles.Initial.Lifetime, B=1)
     Bool2=&SetBoolByFloatComparison(A=$Particles.Age, B=0.12)
    )
    Control Roll On Collision=true
    Friction=1
    Friction During a Bounce=0.25
    Max CPU Trace Length=100
    Module.Control Rest State Via Surface Normal=false
    Module.EnableMaxCollisionCount=false
    Module.Randomize Collision Normal Vector=true
    Randomize Collision Normal=0.25
    ScaleRateOfRotation=false
   ) {enabled=false}
   &SolveForcesAndVelocity(Clamp Velocity=false, Limit Acceleration=false)
   &ScaleMeshSize(Scale Factor=&VectorFromFloat(Value=&FloatFromCurve(FloatCurve=DataInterface() {propertyValues {Curve {keys [<at 0.6 1 RCIM_Cubic>, {arriveTangent=-6.308, interpMode=RCIM_Cubic, leaveTangent=-6.308, tangentMode=RCTM_User, time=1, value=0}]}}})))
   &DynamicMaterialParameters(Module.Param0WriteEnabled=true, Module.Param1WriteEnabled=true, Module.Param2WriteEnabled=false, Module.Param3WriteEnabled=false)
  ]
  rendererClasses [&NiagaraSpriteRendererProperties, &NiagaraMeshRendererProperties]
  renderers [
   {rendererClass=&NiagaraSpriteRendererProperties}
   {rendererClass=&NiagaraMeshRendererProperties, rendererIndex=1}
   {rendererClass=&NiagaraMeshRendererProperties, rendererIndex=2}
   {rendererClass=&NiagaraMeshRendererProperties, rendererIndex=3}
   {rendererClass=&NiagaraMeshRendererProperties, rendererIndex=4}
   {rendererClass=&NiagaraMeshRendererProperties, rendererIndex=5}
   {rendererClass=&NiagaraMeshRendererProperties, rendererIndex=6}
   {rendererClass=&NiagaraMeshRendererProperties, rendererIndex=7}
   {rendererClass=&NiagaraMeshRendererProperties, rendererIndex=8}
  ]
 }
 Flash_Elements {
  emitterUpdateScript [
   &EmitterState(
    Enable Distance Culling=true
    Max Distance Response=::KillImmediately
    Module.bUseMaxDistance=true
    Module.bUseMinDistance=false
    Scalability Mode=::NewEnumerator1
   )
   &SpawnBurst_Instantaneous(Spawn Group=3, Use Spawn Probability=false, UseLoopCountLimit=false) {module=SpawnBurst_Instantaneous003}
   &SpawnBurst_Instantaneous(Spawn Group=2, Use Spawn Probability=false, UseLoopCountLimit=false) {module=SpawnBurst_Instantaneous002}
   &SpawnBurst_Instantaneous(Spawn Group=1, Use Spawn Probability=false, UseLoopCountLimit=false) {module=SpawnBurst_Instantaneous001}
   &SpawnBurst_Instantaneous(Use Spawn Probability=false, UseLoopCountLimit=false)
  ]
  particleSpawnScript [
   &InitializeParticle(
    Color=&MergeHelper_ColorSwitch(0=<rgba 1 1 1 1>, 1=<rgba 1 1 1 0.3>, 2=<rgba 1 1 1 1>, 3=<rgba 1 1 1 1>, Index=$Emitter.SpawnGroup, Num=::NewEnumerator2)
    Lifetime=&MergeHelper_FloatSwitch(0=0.2, 1=0.32, 2=0.2, 3=0.3, Index=$Emitter.SpawnGroup, Num=::NewEnumerator2)
    Sprite Rotation Angle=&MergeHelper_FloatSwitch(
     0=&RandomRangeFloat(Minimum=0, Maximum=1)
     1=&RandomRangeFloat(Minimum=-0.125, Maximum=0.125)
     2=&RandomRangeFloat(Minimum=0, Maximum=1)
     3=&RandomRangeFloat(Minimum=0, Maximum=1)
     Index=$Emitter.SpawnGroup
     Num=::NewEnumerator2
    )
    Sprite Rotation Mode=::NewEnumerator1
    Sprite Size Mode=::NewEnumerator3
    Uniform Sprite Size=&Multiply_Float(
     A=&MergeHelper_FloatSwitch(
      0=&UniformRangedFloat(Minimum=100, Maximum=125)
      1=&RandomRangeFloat(Minimum=120, Maximum=130)
      2=200
      3=&Multiply_Float(A=140, B=&RandomRangeFloat(Minimum=0.7, Maximum=1))
      Index=$Emitter.SpawnGroup
      Num=::NewEnumerator2
     )
     B=$System.FrontEndScale
    )
    UsePositionOffset=false
    Write Material Random=true
   )
   &CameraOffset(Camera Offset Amount=&MergeHelper_FloatSwitch(0=10, 1=30, 2=25, 3=10, Index=$Emitter.SpawnGroup, Num=::NewEnumerator2))
   /SetVariables(Particles.VisibilityTag=$Emitter.SpawnGroup) {isSetParameters=true}
   /SetVariables() {isSetParameters=true}
   /SetVariables(Particles.DynamicParameterAlt2:Vector4f=&MakeVector4(X=1, Y=1, Z=0, W=0)) {isSetParameters=true}
  ]
  particleUpdateScript [
   &ParticleState()
   &ScaleSpriteSize(
    Scale Sprite Size Mode=::NewEnumerator0
    Uniform Scale Factor=&MergeHelper_FloatSwitch(
     0=&FloatFromCurve(FloatCurve=DataInterface() {propertyValues {Curve {keys [
          {arriveTangent=3.6002, interpMode=RCIM_Cubic, leaveTangent=3.6002, tangentMode=RCTM_User, time=0, value=0}
          {arriveTangent=4.9877, interpMode=RCIM_Cubic, leaveTangent=4.9877, tangentMode=RCTM_User, time=0.1858, value=0.4523}
          {arriveTangent=2.6992, interpMode=RCIM_Cubic, leaveTangent=2.6992, tangentMode=RCTM_User, time=0.1973, value=0.9879}
          {arriveTangent=-2.7488, interpMode=RCIM_Cubic, leaveTangent=-2.7488, tangentMode=RCTM_User, time=0.4805, value=0.7971}
          {arriveTangent=-1.5344, interpMode=RCIM_Cubic, leaveTangent=-1.5344, time=0.5064, value=0.5103}
          {arriveTangent=-0.746, tangentMode=RCTM_User, time=1, value=0}
         ]}}})
     1=1
     2=&FloatFromCurve(FloatCurve=DataInterface() {propertyValues {Curve {keys [
          {arriveTangent=1.6131, interpMode=RCIM_Cubic, leaveTangent=1.6131, tangentMode=RCTM_User, time=0, value=0.7}
          {arriveTangent=0.0224, interpMode=RCIM_Cubic, leaveTangent=0.0224, tangentMode=RCTM_User, time=1, value=0}
         ]}}})
     3=&FloatFromCurve(FloatCurve=DataInterface() {propertyValues {Curve {keys [{arriveTangent=3.9086, interpMode=RCIM_Cubic, leaveTangent=3.9086, tangentMode=RCTM_User, time=0, value=0}, <at 1 1.4 RCIM_Cubic>]}}})
     Index=$Particles.VisibilityTag
     Num=::NewEnumerator2
    )
   ) {module=ScaleSpriteSize001}
   &ScaleColor(
    Scale Alpha=&MergeHelper_FloatSwitch(
     0=1
     1=&FloatFromCurve(FloatCurve=DataInterface() {propertyValues {Curve {keys [<at 0 1>,<at 0.8091 0.9932>,<at 1 0>]}}}, Scale Curve=2)
     2=$System.ToDScalar
     3=&FloatFromCurve(
      FloatCurve=DataInterface() {propertyValues {Curve {keys [
          {arriveTangent=-0.2921, interpMode=RCIM_Cubic, leaveTangent=-0.2921, tangentMode=RCTM_User, time=0, value=7}
          {arriveTangent=-7, interpMode=RCIM_Cubic, leaveTangent=-7, time=0.4, value=4}
          <at 1 0 RCIM_Cubic>
         ]}}}
      Scale Curve=2
     )
     Index=$Particles.VisibilityTag
     Num=::NewEnumerator2
    )
    Scale RGB=&MergeHelper_VectorSwitch(
     0=<1 1 1>
     1=&MakeVectorFromLinearColorRGB(Color=<rgba 1 1 1 1>)
     2=&MakeVectorFromLinearColorRGB(Color=&ColorFromCurve(ColorCurve=DataInterface() {propertyValues {
         AlphaCurve {keys [<at 0 1 RCIM_Cubic>]}
         BlueCurve {keys [<at 0 1>,<at 0.7773 0>]}
         GreenCurve {keys [<at 0 1>,<at 0.7773 0>]}
         RedCurve {keys [<at 0 1>,<at 0.7773 0>]}
        }}))
     3=<1 1 1>
     Index=$Particles.VisibilityTag
     Num=::NewEnumerator2
    )
    ScaleA=true
    ScaleRGB=true
   )
   /SetVariables(
    Particles.InnerColor:LinearColor=<rgba 0 0 0 0.6>
    Particles.InnerSize:Vector2f=&Multiply_Vector2D_ByFloat(Float=0.4397, Vector2D=$Particles.SpriteSize)
    Particles.MiddleColor:LinearColor=<rgba 2 2 2 2>
    Particles.MiddleDynamicParameter:Vector4f=<0.1 0 0 0>
    Particles.MiddleSize:Vector2f=&Multiply_Vector2D_ByFloat(Float=0.6435, Vector2D=$Particles.SpriteSize)
    Particles.OuterColor:LinearColor=<rgba 0 0 0 2>
    Particles.OuterDynamicParameter:Vector4f=<0.5 0 0 0>
    Particles.OuterSize:Vector2f=&Multiply_Vector2D_ByFloat(Float=1.0893, Vector2D=$Particles.SpriteSize)
    Particles.SubImageIndex=&RandomRangeFloat(Minimum=0, Maximum=4)
    Particles.SubImageIndexAlt:NiagaraFloat=&RandomRangeFloat(Minimum=0, Maximum=3)
   ) {isSetParameters=true}
   /SetVariables(Particles.DynamicParameterAlt1:Vector4f=&MakeVector4(
     X=&FloatFromCurve(FloatCurve=DataInterface() {propertyValues {Curve {keys [<at 0 0>,<at 1 1>]}}})
     Y=&UniformAOrBFloat(A=0, Evaluation Type=::NewEnumerator0, Recalculate Random Each Loop=true, Override Seed=false)
     Z=0
     W=0
    )) {isSetParameters=true}
   /SetVariables(Particles.DynamicParameterAlt3:Vector4f=&MakeVector4(
     X=&FloatFromCurve(FloatCurve=DataInterface() {propertyValues {Curve {keys [<at 0 1 RCIM_Cubic>, {arriveTangent=-2.742, interpMode=RCIM_Cubic, leaveTangent=-2.742, tangentMode=RCTM_User, time=1, value=0}]}}})
     Y=&FloatFromCurve(FloatCurve=DataInterface() {propertyValues {Curve {keys [<at 0 0 RCIM_Cubic>, {arriveTangent=2.4833, interpMode=RCIM_Cubic, leaveTangent=2.4833, tangentMode=RCTM_User, time=1, value=1}]}}})
     Z=0
     W=0
    )) {isSetParameters=true}
  ]
  rendererClasses [&NiagaraSpriteRendererProperties]
  renderers [
   {rendererClass=&NiagaraSpriteRendererProperties}
   {rendererClass=&NiagaraSpriteRendererProperties, rendererIndex=1}
   {rendererClass=&NiagaraSpriteRendererProperties, rendererIndex=2}
   {rendererClass=&NiagaraSpriteRendererProperties, rendererIndex=3}
   {rendererClass=&NiagaraSpriteRendererProperties, rendererIndex=4}
   {rendererClass=&NiagaraSpriteRendererProperties, rendererIndex=5}
  ]
 }
 Notes {
  emitterUpdateScript [
   &EmitterState(Life Cycle Mode=::NewEnumerator1, Loop Behavior=::NewEnumerator2, Loop Count=3, Loop Duration=0.09, UseLoopDelay=false)
   &SpawnBurst_Instantaneous(
    Loop Count Limit=3
    Spawn Count=&MakeCustomIntFromBool(
     Bool=&SetBoolByFloatComparison(A=&RandomRangeFloat(Minimum=0, Maximum=1), B=0.7)
     True INT=2
     False INT=1
    )
    Use Spawn Probability=false
    UseLoopCountLimit=true
   )
  ]
  particleSpawnScript [
   /SetVariables(Particles.RandomScale:NiagaraFloat=&RandomRangeFloat(Minimum=0, Maximum=1)) {isSetParameters=true}
   &PartitionParticles()
   &InitializeParticle(
    Color=&MaskLinearColor(
     Mask Source=1
     Mask Parameter=$Particles.PartitionParticles.Partition
     Mask=0
     Masked Linear Color=<rgba 0 0 0 1>
     Passthrough Linear Color=<rgba 2 2 2 1>
    )
    Lifetime=&UniformRangedFloat(Minimum=0.5, Maximum=0.8)
    Mesh Scale Mode=::NewEnumerator3
    Mesh Uniform Scale=&Multiply_Float(A=&Lerp_Float(A=0.6, B=0.7, Alpha=$Particles.RandomScale), B=$System.FrontEndScale)
    Sprite Size Mode=::NewEnumerator3
    Uniform Sprite Size=&Multiply_Float(A=&Lerp_Float(A=13, B=20, Alpha=$Particles.RandomScale), B=$System.FrontEndScale)
    UsePositionOffset=false
    Write Material Random=true
   )
   &ShapeLocation(
    MaskSpawnGroup=false
    Override Seed=false
    Sphere Radius=&Multiply_Float(A=15, B=$System.FrontEndScale)
    Sphere Surface Distribution=1
    UseOwnerScale=true
   )
   &AddVelocity(
    Module.Use Falloff Distance=false
    Override Seed=false
    Velocity Mode=::NewEnumerator1
    Velocity Speed=&Multiply_Float(A=&UniformRangedFloat(Minimum=70, Maximum=300), B=$System.FrontEndScale)
   )
   &CameraOffset(Camera Offset Amount=25)
   /SetVariables(Particles.SubImageIndex=&RandomRangeFloat(Minimum=0, Maximum=5)) {isSetParameters=true}
  ]
  particleUpdateScript [
   &ParticleState()
   /Niagara/Modules/Update/Forces/CurlNoiseForce(Noise Frequency=0.02, Noise Quality / Cost=::NewEnumerator0, Noise Strength=100, Pan Noise=false)
   &Drag(Drag=5, Use Linear Drag=true, Use Rotational Drag=false)
   &SolveForcesAndVelocity(Clamp Velocity=false, Limit Acceleration=false)
   &DynamicMaterialParameters(
    Index 0 Param 1=0.05
    Index 0 Param 2=0.05
    Module.Index0Float0Write=true
    Module.Index0Float1Write=true
    Module.Index0Float2Write=false
    Module.Index0Float3Write=false
   )
   &ScaleSpriteSize(
    Scale Factor=&Multiply_Vector2D_ByFloat(
     Float=&FloatFromCurve(FloatCurve=DataInterface() {propertyValues {Curve {keys [
          <at 0 0.5 RCIM_Cubic>
          {arriveTangent=-0.0339, interpMode=RCIM_Cubic, leaveTangent=-0.0339, tangentMode=RCTM_User, time=0.2, value=1}
          {arriveTangent=2.6545, interpMode=RCIM_Cubic, leaveTangent=2.6545, tangentMode=RCTM_User, time=0.5718, value=1.1441}
          {arriveTangent=-1.585, interpMode=RCIM_Cubic, leaveTangent=-1.585, tangentMode=RCTM_User, time=0.7, value=1.45}
          {arriveTangent=-6.5554, interpMode=RCIM_Cubic, leaveTangent=-6.5554, time=0.8391, value=0.8815}
          {arriveTangent=-6.0448, interpMode=RCIM_Cubic, leaveTangent=-6.0448, time=0.85, value=0.4667}
          {arriveTangent=-4.0432, interpMode=RCIM_Cubic, leaveTangent=-4.0432, tangentMode=RCTM_User, time=0.985, value=0}
         ]}}})
     Vector2D=/Niagara/DynamicInputs/Add/Add_Vector2(
      A:Vector2f=<1 1>
      B:Vector2f=&Vector2DFromCurve(
       Vector2Curve=DataInterface() {propertyValues {
         XCurve {keys [
           {arriveTangent=-0.5303, interpMode=RCIM_Cubic, leaveTangent=-0.5303, tangentMode=RCTM_User, time=0.177, value=0}
           {arriveTangent=-0.2018, interpMode=RCIM_Cubic, leaveTangent=-0.2018, time=0.3799, value=0.1558}
           {arriveTangent=-0.2769, interpMode=RCIM_Cubic, leaveTangent=-0.2769, time=0.6837, value=-0.1023}
           {arriveTangent=-0.0478, interpMode=RCIM_Cubic, leaveTangent=-0.0478, tangentMode=RCTM_User, time=0.9425, value=0}
          ]}
         YCurve {keys [
           {arriveTangent=-0.0378, interpMode=RCIM_Cubic, leaveTangent=-0.0378, tangentMode=RCTM_User, time=0.1174, value=0}
           {arriveTangent=0.2099, interpMode=RCIM_Cubic, leaveTangent=0.2099, time=0.4023, value=-0.1196}
           {arriveTangent=0.248, interpMode=RCIM_Cubic, leaveTangent=0.248, time=0.6894, value=0.1201}
           {arriveTangent=-0.1316, interpMode=RCIM_Cubic, leaveTangent=-0.1316, tangentMode=RCTM_User, time=0.8846, value=0}
          ]}
        }}
       ScaleCurve=&RandomRangeVector2D(Minimum=<0.2 0.2>, Maximum=<1.5 1.5>)
      )
     )
    )
    Scale Sprite Size Mode=::NewEnumerator2
   )
   /SetVariables(
    Particles.DynamicParameterAlt3:Vector4f=&MakeVector4(
     X=1
     Y=&FloatFromCurve(FloatCurve=DataInterface() {propertyValues {Curve {keys [<at 0.8 0.2>,<at 1 1>]}}}, Scale Curve=0.5)
     Z=0
     W=0
    )
    Particles.LinesCol:LinearColor=&MakeLinearColorFromVectorAndFloat("Vector (RGB)"=&MakeVectorFromLinearColorRGB(Color=$Particles.Initial.Color), "Float (Alpha)"=15)
    Particles.LinesRot:NiagaraFloat=&RandomRangeFloat(Minimum=0, Maximum=360)
    Particles.LinesRot2:NiagaraFloat=&RandomRangeFloat(Minimum=0, Maximum=360)
    Particles.LinesSize:Vector2f=&Multiply_Vector2D_ByFloat(
     Float=&FloatFromCurve(FloatCurve=DataInterface() {propertyValues {Curve {keys [<at 0.8 0>,<at 0.85 1>]}}}, Scale Curve=2)
     Vector2D=$Particles.Initial.SpriteSize
    )
    Particles.LinesSize2:Vector2f=&Multiply_Vector2D_ByFloat(
     Float=&FloatFromCurve(FloatCurve=DataInterface() {propertyValues {Curve {keys [<at 0.8 0>,<at 0.85 1>]}}}, Scale Curve=2)
     Vector2D=$Particles.Initial.SpriteSize
    )
   ) {isSetParameters=true}
  ]
  rendererClasses [&NiagaraSpriteRendererProperties]
  renderers [
   {rendererClass=&NiagaraSpriteRendererProperties}
   {rendererClass=&NiagaraSpriteRendererProperties, rendererIndex=1}
  ]
 }
 Toon_Sparks {
  emitterUpdateScript [
   &EmitterState()
   &SpawnBurst_Instantaneous(
    Spawn Count=&RandomRangeInt(Minimum=12, Maximum=15)
    Use Spawn Probability=false
    UseLoopCountLimit=false
   )
  ]
  particleSpawnScript [
   &InitializeParticle(
    Color=&MakeCustomLinearColorFromBool(Bool=&RandomBool(), True Color=<rgba 2 2 2 1>, False Color=<rgba 0 0 0 1>)
    Lifetime=&MergeHelper_FloatSwitch(
     0=&RandomRangeFloat(Minimum=0.24, Maximum=0.34)
     1=&RandomRangeFloat(Minimum=0.1, Maximum=0.2)
     2=?
     3=?
     Index=$Emitter.SpawnGroup
     Num=::NewEnumerator0
    )
    Sprite Rotation Angle=&MergeHelper_FloatSwitch(
     0=0
     1=&RandomRangeFloat(Minimum=0, Maximum=1)
     2=?
     3=?
     Index=$Emitter.SpawnGroup
     Num=::NewEnumerator0
    )
    Sprite Rotation Mode=::NewEnumerator1
    Sprite Size=&Multiply_Vector2D_ByFloat(
     Float=&FN_FrontEndFloatModifier(Gameplay=1, Front End=0.45)
     Vector2D=&RandomRangeVector2D(Minimum=<1.9 3.1487>, Maximum=<3 9.8129>)
    )
    Sprite Size Mode=::NewEnumerator4
    UsePositionOffset=false
    Write Material Random=true
   )
   &AddVelocityFromPoint(
    Module.Use Falloff Distance=false
    Velocity Strength=&Multiply_Float(
     A=&MergeHelper_FloatSwitch(
      0=&RandomRangeFloat(Minimum=200, Maximum=400)
      1=140
      2=?
      3=?
      Index=$Emitter.SpawnGroup
      Num=::NewEnumerator0
     )
     B=$System.FrontEndScale
    )
   )
   &AddVelocity(
    Velocity=&MergeHelper_VectorSwitch(
     0=&RandomRangeVector(Minimum=<100 0 0>, Maximum=<200 0 0>)
     1=<0 0 0>
     2=?
     3=?
     Index=$Emitter.SpawnGroup
     Num=::NewEnumerator0
    )
    Velocity Speed Scale=$System.FrontEndScale
   )
  ]
  particleUpdateScript [
   &ParticleState()
   &ScaleSpriteSize(
    Scale Factor=&Vector2DFromFloat(Value=&FloatFromCurve(FloatCurve=DataInterface() {propertyValues {Curve {keys [
          {arriveTangent=18.3035, interpMode=RCIM_Cubic, leaveTangent=18.3035, tangentMode=RCTM_User, time=0, value=0}
          {interpMode=RCIM_Cubic, tangentMode=RCTM_User, time=0.118, value=1}
          {arriveTangent=-1.8884, interpMode=RCIM_Cubic, leaveTangent=-1.8884, tangentMode=RCTM_User, time=0.5494, value=0.6735}
          {interpMode=RCIM_Cubic, tangentMode=RCTM_User, time=1, value=0}
         ]}}}))
    Scale Sprite Size Mode=::NewEnumerator2
   )
   &ScaleColor(ScaleA=true, ScaleRGB=true)
   &Drag(
    Drag=&FloatFromCurve(FloatCurve=DataInterface() {propertyValues {Curve {keys [<at 0 1>,<at 1 6>]}}}, Scale Curve=1.2)
    Use Linear Drag=true
    Use Rotational Drag=true
   )
   /Niagara/Modules/Update/Forces/V2/CurlNoiseForce(Bias Noise=false, Noise Strength=&RandomRangeFloat(Minimum=200, Maximum=400), Pan Noise=false)
   &AccelerationForce(Acceleration=<0 0 -240>)
   &SolveForcesAndVelocity(Clamp Velocity=false, Limit Acceleration=false)
   &ScaleSpriteSizeBySpeed(Max Scale Factor=<1 6>, Min Scale Factor=<1 3>)
  ]
  rendererClasses [&NiagaraSpriteRendererProperties]
  renderers [{rendererClass=&NiagaraSpriteRendererProperties}]
 }
}
summary {
 systemName=NS_Pickaxe_Bass_BombaToon_Impact
 emitters [
  {
   emitterName=FN_PhysMatDebris
   rendererClasses [&NiagaraSpriteRendererProperties, &NiagaraMeshRendererProperties]
  }
  {emitterName=Flash_Elements, rendererClasses [&NiagaraSpriteRendererProperties]}
  {emitterName=Toon_Sparks, rendererClasses [&NiagaraSpriteRendererProperties]}
  {emitterName=Notes, rendererClasses [&NiagaraSpriteRendererProperties]}
 ]
 userVariables [
  User.bIsFrontEnd {type=NiagaraBool}
  User.bIsFrontEndPreview {type=NiagaraBool}
  User.PhysicalSurface {type=NiagaraInt32}
 ]
}
