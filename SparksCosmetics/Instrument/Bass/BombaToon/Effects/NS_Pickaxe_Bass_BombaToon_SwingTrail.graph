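# Niagara swing trail for the BombaToon bass pickaxe, built from six short-lived ribbon emitters on anim-trail modules.
# Main, glow and black ribbon layers span socket indices 1 and 2, while the L/R side trails span 5 and 6 in opposite order.
# System FadeSpawn sets initial alpha and FadeSystem scales it during update, with an extra alpha curve on the glow.
# User parameters expose skeletal mesh, swing section index and ribbon ID offset.
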NiagaraSystem /SparksCosmetics/Instrument/Bass/BombaToon/Effects/NS_Pickaxe_Bass_BombaToon_SwingTrail
dependencies {
 usedDataInterfaces [&NiagaraDataInterfaceColorCurve, &NiagaraDataInterfaceSocketReader]
 usedDynamicInputs [
  &MergeHelper_FloatSwitch
  &ClampFloat
  &RemapRange
  &Add_Float
  &Subtract_Float
 ]
 usedModules [
  &self:SystemSpawnScript.SystemScriptSource.SystemScriptGraph.NiagaraNodeAssignment_1.SetVariables_#1
  &self:SystemSpawnScript.SystemScriptSource.SystemScriptGraph.NiagaraNodeAssignment_2.SetVariables_#2
  &self:SystemSpawnScript.SystemScriptSource.SystemScriptGraph.NiagaraNodeAssignment_5.SetVariables_#3
  &self:SystemSpawnScript.SystemScriptSource.SystemScriptGraph.NiagaraNodeAssignment_0.SetVariables_#4
  &SystemState
  &self:SystemSpawnScript.SystemScriptSource.SystemScriptGraph.NiagaraNodeAssignment_3.SetVariables_#5
 ]
}
emitters {
 TrailGlow {
  emitterUpdateScript [
   &EmitterState()
   &FN_SpawnAnimTrail()
  ]
  particleSpawnScript [
   &FN_InitializeAnimTrailBoneIndecies(First Bone or Socket Index=1, Second Bone or Socket Index=2)
   &FN_InitializeAnimTrail(
    AnimTrailWidth=0.744
    Color=&MakeLinearColorFromVectorAndFloat(
     "Vector (RGB)"=&MakeVectorFromLinearColorRGB(Color=<rgba 1 1 1 1>)
     "Float (Alpha)"=&Multiply_Float(A=0.23, B=$System.FadeSpawn)
    )
    Lifetime=0.09
   )
   /SetVariables() {isSetParameters=true}
   &AddVelocity(Velocity=<2000 0 0>) {enabled=false}
  ]
  particleUpdateScript [
   &FN_UpdateAnimTrail() {module=FN_UpdateAnimTrail001}
   &ScaleColor(
    Scale Mode=1
    Scale RGBA=&Multiply_Vector4(A=<1 1 1 1>, B=&MakeVector4FromVectorAndFloat("Vector (RGB)"=<1 1 1>, "Float (Alpha)"=$System.FadeSystem))
    ScaleRGBA=true
   ) {module=ScaleColor002}
   &ScaleColor(
    Linear Color Curve=DataInterface() {propertyValues {
      AlphaCurve {keys [<at 0.0929 0>,<at 0.1838 0.1345>,<at 0.5701 1>]}
      BlueCurve {keys [<at 0 1>]}
      GreenCurve {keys [<at 0 1>]}
      RedCurve {keys [<at 0 1>]}
     }}
    Scale Mode=2
   ) {module=ScaleColor001}
   &SolveForcesAndVelocity(Clamp Velocity=false, Limit Acceleration=false)
  ]
  rendererClasses [&NiagaraRibbonRendererProperties]
  renderers [{rendererClass=&NiagaraRibbonRendererProperties}]
 }
 TrailMain {
  emitterUpdateScript [
   &EmitterState()
   &FN_SpawnAnimTrail()
  ]
  particleSpawnScript [
   &FN_InitializeAnimTrailBoneIndecies(First Bone or Socket Index=1, Second Bone or Socket Index=2)
   &FN_InitializeAnimTrail(
    Color=&MakeLinearColorFromVectorAndFloat(
     "Vector (RGB)"=&MakeVectorFromLinearColorRGB(Color=<rgba 1 1 1 1>)
     "Float (Alpha)"=&Multiply_Float(A=1.5, B=$System.FadeSpawn)
    )
    Lifetime=0.12
   )
   /SetVariables() {isSetParameters=true}
   &AddVelocity(Velocity=<2000 0 0>) {enabled=false}
  ]
  particleUpdateScript [
   &FN_UpdateAnimTrail() {module=FN_UpdateAnimTrail001}
   &ScaleColor(
    Scale Mode=1
    Scale RGBA=&Multiply_Vector4(A=<1 1 1 1>, B=&MakeVector4FromVectorAndFloat("Vector (RGB)"=<1 1 1>, "Float (Alpha)"=$System.FadeSystem))
    ScaleRGBA=true
   ) {module=ScaleColor002}
   /SetVariables(Particles.LineWidth=48.7173) {isSetParameters=true, module=SetVariables_#6}
   &SolveForcesAndVelocity(Clamp Velocity=false, Limit Acceleration=false)
  ]
  rendererClasses [&NiagaraRibbonRendererProperties]
  renderers [{rendererClass=&NiagaraRibbonRendererProperties}]
 }
 TrailMain001 {
  emitterUpdateScript [
   &EmitterState()
   &FN_SpawnAnimTrail()
  ]
  particleSpawnScript [
   &FN_InitializeAnimTrailBoneIndecies(First Bone or Socket Index=1, Second Bone or Socket Index=2)
   &FN_InitializeAnimTrail(
    Color=&MakeLinearColorFromVectorAndFloat(
     "Vector (RGB)"=&MakeVectorFromLinearColorRGB(Color=<rgba 1 1 1 1>)
     "Float (Alpha)"=&Multiply_Float(A=1.5, B=$System.FadeSpawn)
    )
    Lifetime=0.08
   )
   /SetVariables() {isSetParameters=true}
   &AddVelocity(Velocity=<2000 0 0>) {enabled=false}
  ]
  particleUpdateScript [
   &FN_UpdateAnimTrail() {module=FN_UpdateAnimTrail001}
   &ScaleColor(
    Scale Mode=1
    Scale RGBA=&Multiply_Vector4(A=<1 1 1 1>, B=&MakeVector4FromVectorAndFloat("Vector (RGB)"=<1 1 1>, "Float (Alpha)"=$System.FadeSystem))
    ScaleRGBA=true
   ) {module=ScaleColor002}
   /SetVariables(Particles.LineWidth=48.7173) {isSetParameters=true, module=SetVariables_#6}
   &SolveForcesAndVelocity(Clamp Velocity=false, Limit Acceleration=false)
  ]
  rendererClasses [&NiagaraRibbonRendererProperties]
  renderers [{rendererClass=&NiagaraRibbonRendererProperties}]
 }
 TrailRibbon {
  emitterUpdateScript [
   &EmitterState()
   &FN_SpawnAnimTrail()
  ]
  particleSpawnScript [
   &FN_InitializeAnimTrailBoneIndecies(First Bone or Socket Index=1, Second Bone or Socket Index=2)
   &FN_InitializeAnimTrail(
    Color=&MakeLinearColorFromVectorAndFloat(
     "Vector (RGB)"=&MakeVectorFromLinearColorRGB(Color=<rgba 0 0 0 1>)
     "Float (Alpha)"=&Multiply_Float(A=1.5, B=$System.FadeSpawn)
    )
    Lifetime=0.1
   )
   /SetVariables() {isSetParameters=true}
   &AddVelocity(Velocity=<2000 0 0>) {enabled=false}
  ]
  particleUpdateScript [
   &FN_UpdateAnimTrail() {module=FN_UpdateAnimTrail001}
   &ScaleColor(
    Scale Mode=1
    Scale RGBA=&Multiply_Vector4(A=<1 1 1 1>, B=&MakeVector4FromVectorAndFloat("Vector (RGB)"=<1 1 1>, "Float (Alpha)"=$System.FadeSystem))
    ScaleRGBA=true
   ) {module=ScaleColor002}
   &SolveForcesAndVelocity(Clamp Velocity=false, Limit Acceleration=false)
  ]
  rendererClasses [&NiagaraRibbonRendererProperties]
  renderers [{rendererClass=&NiagaraRibbonRendererProperties}]
 }
 TrailSide_L {
  emitterUpdateScript [
   &EmitterState()
   &FN_SpawnAnimTrail()
  ]
  particleSpawnScript [
   &FN_InitializeAnimTrailBoneIndecies(First Bone or Socket Index=6, Second Bone or Socket Index=5)
   &FN_InitializeAnimTrail(
    Color=&MakeLinearColorFromVectorAndFloat(
     "Vector (RGB)"=&MakeVectorFromLinearColorRGB(Color=<rgba 1 1 1 1>)
     "Float (Alpha)"=&Multiply_Float(A=1.5, B=$System.FadeSpawn)
    )
    Lifetime=0.1
   )
   /SetVariables(Particles.RibbonFacing=<0 0 -1>) {isSetParameters=true}
   &AddVelocity(Velocity=<2000 0 0>) {enabled=false}
  ]
  particleUpdateScript [
   &FN_UpdateAnimTrail() {module=FN_UpdateAnimTrail001}
   &ScaleColor(
    Scale Mode=1
    Scale RGBA=&Multiply_Vector4(A=<1 1 1 1>, B=&MakeVector4FromVectorAndFloat("Vector (RGB)"=<1 1 1>, "Float (Alpha)"=$System.FadeSystem))
    ScaleRGBA=true
   ) {module=ScaleColor002}
   &SolveForcesAndVelocity(Clamp Velocity=false, Limit Acceleration=false)
  ]
  rendererClasses [&NiagaraRibbonRendererProperties]
  renderers [{rendererClass=&NiagaraRibbonRendererProperties}]
 }
 TrailSide_R {
  emitterUpdateScript [
   &EmitterState()
   &FN_SpawnAnimTrail()
  ]
  particleSpawnScript [
   &FN_InitializeAnimTrailBoneIndecies(First Bone or Socket Index=5, Second Bone or Socket Index=6)
   &FN_InitializeAnimTrail(
    Color=&MakeLinearColorFromVectorAndFloat(
     "Vector (RGB)"=&MakeVectorFromLinearColorRGB(Color=<rgba 1 1 1 1>)
     "Float (Alpha)"=&Multiply_Float(A=1.5, B=$System.FadeSpawn)
    )
    Lifetime=0.1
   )
   /SetVariables() {isSetParameters=true}
   &AddVelocity(Velocity=<2000 0 0>) {enabled=false}
  ]
  particleUpdateScript [
   &FN_UpdateAnimTrail() {module=FN_UpdateAnimTrail001}
   &ScaleColor(
    Scale Mode=1
    Scale RGBA=&Multiply_Vector4(A=<1 1 1 1>, B=&MakeVector4FromVectorAndFloat("Vector (RGB)"=<1 1 1>, "Float (Alpha)"=$System.FadeSystem))
    ScaleRGBA=true
   ) {module=ScaleColor002}
   &SolveForcesAndVelocity(Clamp Velocity=false, Limit Acceleration=false)
  ]
  rendererClasses [&NiagaraRibbonRendererProperties]
  renderers [{rendererClass=&NiagaraRibbonRendererProperties}]
 }
}
summary {
 systemName=NS_Pickaxe_Bass_BombaToon_SwingTrail
 emitters [
  {emitterName=TrailMain, rendererClasses [&NiagaraRibbonRendererProperties]}
  {emitterName=TrailSide_L, rendererClasses [&NiagaraRibbonRendererProperties]}
  {emitterName=TrailRibbon, rendererClasses [&NiagaraRibbonRendererProperties]}
  {emitterName=TrailGlow, rendererClasses [&NiagaraRibbonRendererProperties]}
  {emitterName=TrailSide_R, rendererClasses [&NiagaraRibbonRendererProperties]}
  {emitterName=TrailMain001, rendererClasses [&NiagaraRibbonRendererProperties]}
 ]
 userVariables [
  User.RibbonIDOffset {type=NiagaraInt32}
  User.SkeletalMesh {
   type=NiagaraDataInterfaceSkeletalMesh
   defaultValue {
    struct=NiagaraExt_VariableValue_DataInterface
    value {dataInterface=&self:NiagaraDataInterfaceSkeletalMesh_0}
   }
  }
  User.SwingSectionIndex {type=NiagaraInt32}
 ]
}
