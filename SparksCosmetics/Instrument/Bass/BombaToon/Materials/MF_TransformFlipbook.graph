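# Material function for the BombaToon bass cosmetic that offsets, rotates and scales UVs to place one flipbook cell on a surface.
# It shifts UVs, spins them with CustomRotator using degree input, then scales around the half-cell center by OffsetScale times the larger grid dimension.
# Index selects the column and row offset via Fmod and Floor for Flipbook Coordinates.
# It also outputs the unshifted General Coordinates and a Mask that clips to a single cell.
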MaterialFunction /SparksCosmetics/Instrument/Bass/BombaToon/Materials/MF_TransformFlipbook
graph {nodes [
  n1 = TextureCoordinate()
  n2 = Constant()
  n3 = Constant()
  n4 = Divide(B=AppendVector(A=NamedRerouteUsage(), B=NamedRerouteUsage())) {constA=-1}
  n5 = Add(
   A=AppendVector(
    A=FunctionInput(Preview=@n2) {bUsePreviewValueAsDefault=true, inputName=OffsetU, inputType=FunctionInput_Scalar, sortPriority=1}
    B=FunctionInput(Preview=@n3) {bUsePreviewValueAsDefault=true, inputName=OffsetV, inputType=FunctionInput_Scalar, sortPriority=2}
   )
   B=Divide(A=Add(A=@n4))
  )
  n6 = MaterialFunctionCall(
   UVs=Subtract(A=FunctionInput(Preview=@n1) {bUsePreviewValueAsDefault=true, inputName=UVs, inputType=FunctionInput_Vector2, sortPriority=-1}, B=@n5)
   Rotation Center=NamedRerouteUsage()
   "Rotation Angle (0-1)"=NamedRerouteUsage()
  ) {materialFunction=&CustomRotator}
  n7 = NamedRerouteDeclaration(None=FunctionInput(Preview=Constant() {r=4}) {bUsePreviewValueAsDefault=true, inputName=Number of Columns, inputType=FunctionInput_Scalar, sortPriority=3}) {name=Columns, nodeColor=<rgba 0.1719 0 1 1>}
  n8 = NamedRerouteDeclaration(None=FunctionInput(Preview=Constant() {r=4}) {bUsePreviewValueAsDefault=true, inputName=Number of Rows, inputType=FunctionInput_Scalar, sortPriority=4}) {name=Rows, nodeColor=<rgba 0.1719 0 1 1>}
  n9 = AppendVector(A=@n7, B=@n8)
  n10 = Divide(A=Divide(B=@n9))
  n11 = FunctionInput(Preview=Constant() {r=1}) {bUsePreviewValueAsDefault=true, inputName=OffsetScale, inputType=FunctionInput_Vector2}
  n12 = If(A=@n7, B=@n8, "A > B"=@n7, "A < B"=@n8)
  n13 = Add(A=Divide(A=Subtract(A=@n6.Rotated Values, B=@n10) {constB=0.25}, B=Multiply(A=@n11, B=@n12)), B=@n10) {constB=0.25}
  n14 = FunctionInput(Preview=Constant()) {bUsePreviewValueAsDefault=true, inputName=Index, inputType=FunctionInput_Vector2, sortPriority=6}
  n15 = NamedRerouteUsage()
  n16 = Divide(A=Floor(None=Fmod(A=@n14, B=@n15)), B=Floor(None=@n15))
  n17 = NamedRerouteUsage()
  n18 = NamedRerouteUsage()
  n19 = Fmod(A=@n14, B=Multiply(A=@n17, B=@n18))
  n20 = Divide(A=Floor(None=Divide(A=@n19, B=@n17)), B=Floor(None=@n18))
  FunctionOutput(None=Add(A=FunctionOutput(None=@n13) {outputName=General Coordinates, sortPriority=4}, B=AppendVector(A=@n16, B=@n20)) {constB=3}) {outputName=Flipbook Coordinates, sortPriority=3}
  n21 = NamedRerouteUsage()
  n22 = NamedRerouteUsage()
  n23 = Floor(None=Multiply(A=@n13, B=AppendVector(A=@n21, B=@n22)) {constB=2})
  n24 = OneMinus(None=Clamp(None=Abs(None=@n23)))
  FunctionOutput(None=Multiply(A=ComponentMask(None=@n24) {g=true}, B=ComponentMask(None=@n24) {r=true})) {outputName=Mask, sortPriority=4}
  NamedRerouteDeclaration(None=@n5) {name=Offset, nodeColor=<rgba 0 0.25 1 1>}
  NamedRerouteDeclaration(None=Divide(A=Divide(B=@n9))) {name=RotationOffset, nodeColor=<rgba 0 1 0.7188 1>}
  n25 = Constant()
  NamedRerouteDeclaration(None=Divide(A=FunctionInput(Preview=@n25) {bUsePreviewValueAsDefault=true, inputName=OffsetRotation, inputType=FunctionInput_Scalar, sortPriority=1}) {constB=360}) {name=Rotation, nodeColor=<rgba 1 0.3984 0 1>}
 ]}
