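# Material function for the BombaToon bass cosmetic that layers an animated cartoon face (mouth and two eyes) onto Color and Shade inputs.
# The mouth is picked from a flipbook by index on UV1, with position, scale, rotation and flip parameters.
# Each eye samples one channel-packed texture for mask and offset pupil, with eyelids cut procedurally by thresholding V.
# Only the left eye's UVs add a BumpOffset.
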MaterialFunction /SparksCosmetics/Instrument/Bass/BombaToon/Materials/MF_BombaToon_Bass
graph {nodes [
  n1 = LinearInterpolate(A=FunctionInput() {inputName=Color, previewValue=<1 1 1 1>}, B=NamedRerouteUsage(), Alpha=NamedRerouteUsage())
  n2 = NamedRerouteUsage()
  n3 = NamedRerouteUsage()
  FunctionOutput(None=LinearInterpolate(A=LinearInterpolate(A=@n1, B=@n2, Alpha=@n3), B=NamedRerouteUsage(), Alpha=NamedRerouteUsage())) {outputName=Result_Color}
  n4 = LinearInterpolate(A=FunctionInput() {inputName=Shade, sortPriority=1, previewValue=<0.7344 0.7344 0.7344 1>}, B=NamedRerouteUsage(), Alpha=NamedRerouteUsage())
  n5 = NamedRerouteUsage()
  n6 = NamedRerouteUsage()
  FunctionOutput(None=LinearInterpolate(A=LinearInterpolate(A=@n4, B=@n5, Alpha=@n6), B=NamedRerouteUsage(), Alpha=NamedRerouteUsage())) {outputName=Result_Shade, sortPriority=1}
  NamedRerouteDeclaration(None=Constant3Vector() {constant=<rgba 0.5711 0.5776 0.5972 1>}) {name=White, nodeColor=<rgba 1 0.4688 0 1>}
  n7 = ScalarParameter() {defaultValue=1, group=Face Flipbook, parameterName=Mouth_Scale_U, sortPriority=2}
  n8 = ScalarParameter() {defaultValue=1, group=Face Flipbook, parameterName=Mouth_Scale_V, sortPriority=2}
  n9 = Saturate(None=Round(None=ScalarParameter() {group=Face Flipbook, parameterName=Mouth_VerticalFlip}))
  n10 = MaterialFunctionCall!(
   UVs=TextureCoordinate() {coordinateIndex=1}
   OffsetScale=Multiply(A=ScalarParameter() {defaultValue=0.6, group=Face Flipbook, parameterName=Offset_Mouth_Scale, sortPriority=2}, B=AppendVector(A=@n7, B=@n8))
   OffsetRotation=Add(A=ScalarParameter() {group=Face Flipbook, parameterName=Mouth_Rotation, sortPriority=2}, B=Multiply(B=@n9) {constA=180})
   OffsetU=Add(
    A=ScalarParameter() {group=Face Flipbook, parameterName=Offset_Mouth_PositionU, sortPriority=2}
    B=ScalarParameter() {group=Face Flipbook, parameterName=Mouth_Position_U, sortPriority=2}
   )
   OffsetV=Add(
    A=ScalarParameter() {group=Face Flipbook, parameterName=Mouth_Position_V, sortPriority=3}
    B=ScalarParameter() {defaultValue=0.3, group=Face Flipbook, parameterName=Offset_Mouth_PositionV, sortPriority=3}
   )
   Number of Columns=Constant() {r=2}
   Number of Rows=Constant() {r=3}
   Index=ScalarParameter() {group=Face Flipbook, parameterName=Mouth_Index, sortPriority=2}
  ) {materialFunction=&/MF_TransformFlipbook}
  n11 = TextureSampleParameter2D!(UVs=@n10.Flipbook Coordinates, Level, Apply View MipBias) {
   group=Face Flipbook
   mipValueMode=TMVM_MipLevel
   parameterName=Flipbook
   sortPriority=1
   texture=/SparksCosmetics/Instrument/Bass/BombaToon/Textures/T_BombaToon_Bass_Flipbook_Mouth
  }
  NamedRerouteDeclaration(None=Multiply(A=@n11.A, B=@n10.Mask)) {name=Mask_Mouth, nodeColor=<rgba 1 0.375 0 1>}
  NamedRerouteDeclaration(None=LinearInterpolate(B=NamedRerouteUsage(), Alpha=@n11.RGB)) {name=Color_Mouth, nodeColor=<rgba 0.0547 0 1 1>}
  n12 = Subtract(
   A=ScalarParameter() {group=Face Flipbook, parameterName=Eyes_Position_U}
   B=ScalarParameter() {defaultValue=0.23, group=Face Flipbook, parameterName=Offset_Eyes_Distance}
  )
  n13 = ScalarParameter() {group=Face Flipbook, parameterName=Eyes_Position_V}
  n14 = Constant()
  n15 = ScalarParameter() {defaultValue=0.2, group=Face Flipbook, parameterName=OffsetEyePositionV}
  n16 = Add(
   A=BumpOffset(Coordinate=TextureCoordinate() {coordinateIndex=1}, Height=Constant() {r=-0.07})
   B=Add(A=AppendVector(A=@n12, B=@n13), B=AppendVector(A=@n14, B=@n15))
  ) {constB=0}
  n17 = ScalarParameter() {defaultValue=0.5, group=Face Flipbook, parameterName=Offset_Eyes_Scale}
  n18 = AppendVector(
   A=ScalarParameter() {defaultValue=1, group=Face Flipbook, parameterName=Eyes_Scale_U}
   B=ScalarParameter() {defaultValue=1, group=Face Flipbook, parameterName=Eyes_Scale_V}
  )
  n19 = Add(A=Divide(A=Subtract(A=@n16) {constB=0.5}, B=Multiply(A=@n17, B=@n18))) {constB=0.5}
  n20 = Saturate(None=@n19)
  n21 = TextureSample!(UVs=@n20, Tex, Level, Apply View MipBias) {mipValueMode=TMVM_MipLevel, texture=/SparksCosmetics/Instrument/Bass/BombaToon/Textures/T_BombaToon_Bass_Flipbook_Eye}
  NamedRerouteDeclaration(None=@n21.B) {name=Mask_EyeL, nodeColor=<rgba 1 0.375 0 1>}
  n22 = Add(
   A=ScalarParameter() {group=Face Flipbook, parameterName=Eyes_Position_U}
   B=ScalarParameter() {defaultValue=0.23, group=Face Flipbook, parameterName=Offset_Eyes_Distance}
  )
  n23 = ScalarParameter() {group=Face Flipbook, parameterName=Eyes_Position_V}
  n24 = Constant()
  n25 = ScalarParameter() {defaultValue=0.2, group=Face Flipbook, parameterName=OffsetEyePositionV}
  n26 = Add(A=TextureCoordinate() {coordinateIndex=1}, B=Add(A=AppendVector(A=@n22, B=@n23), B=AppendVector(A=@n24, B=@n25))) {constB=0}
  n27 = ScalarParameter() {defaultValue=0.5, group=Face Flipbook, parameterName=Offset_Eyes_Scale}
  n28 = AppendVector(
   A=ScalarParameter() {defaultValue=1, group=Face Flipbook, parameterName=Eyes_Scale_U}
   B=ScalarParameter() {defaultValue=1, group=Face Flipbook, parameterName=Eyes_Scale_V}
  )
  n29 = Add(A=Divide(A=Subtract(A=@n26) {constB=0.5}, B=Multiply(A=@n27, B=@n28))) {constB=0.5}
  n30 = Saturate(None=@n29)
  n31 = TextureSample(UVs=@n30) {mipValueMode=TMVM_MipLevel, texture=&T_BombaToon_Bass_Flipbook_Eye}
  NamedRerouteDeclaration(None=@n31.B) {name=Mask_EyeR, nodeColor=<rgba 1 0.375 0 1>}
  n32 = ComponentMask(None=@n19) {g=true}
  n33 = Subtract(A=@n32, B=Multiply(A=ScalarParameter() {defaultValue=0.768, group=Face Flipbook, parameterName=Eyelids_Upper_L}) {constB=0.55})
  n34 = AppendVector(
   A=ScalarParameter() {defaultValue=0.268, group=Face Flipbook, parameterName=Pupils_Position_U}
   B=ScalarParameter() {defaultValue=0.1, group=Face Flipbook, parameterName=Pupils_Position_V}
  )
  n35 = TextureSample(UVs=Add(A=@n20, B=Multiply(A=@n34) {constB=0.1}) {constB=0.4}) {mipValueMode=TMVM_MipLevel, texture=&T_BombaToon_Bass_Flipbook_Eye}
  n36 = Add(A=@n32, B=Multiply(A=ScalarParameter() {group=Face Flipbook, parameterName=Eyelids_Lower_L}) {constB=0.45}) {constB=0.5}
  n37 = Add(A=@n36) {constB=-0.04}
  n38 = LinearInterpolate(
   A=OneMinus(None=Ceil(None=@n33))
   B=LinearInterpolate(A=OneMinus(None=@n35.G), B=Floor(None=@n37), Alpha=Floor(None=@n36)) {constA=0.2}
   Alpha=Ceil(None=Subtract(A=@n33) {constB=0.04})
  ) {constA=0.2, constB=0.2}
  NamedRerouteDeclaration(None=LinearInterpolate(B=NamedRerouteUsage(), Alpha=LinearInterpolate(B=@n38, Alpha=@n21.R))) {name=Color_EyeL, nodeColor=<rgba 0.0547 0 1 1>}
  n39 = ComponentMask(None=@n29) {g=true}
  n40 = Subtract(A=@n39, B=Multiply(A=ScalarParameter() {defaultValue=0.768, group=Face Flipbook, parameterName=Eyelids_Upper_R}) {constB=0.55})
  n41 = AppendVector(
   A=ScalarParameter() {defaultValue=0.268, group=Face Flipbook, parameterName=Pupils_Position_U}
   B=ScalarParameter() {defaultValue=0.1, group=Face Flipbook, parameterName=Pupils_Position_V}
  )
  n42 = TextureSample(UVs=Add(A=@n30, B=Multiply(A=@n41) {constB=0.1}) {constB=0.4}) {mipValueMode=TMVM_MipLevel, texture=&T_BombaToon_Bass_Flipbook_Eye}
  n43 = Add(A=@n39, B=Multiply(A=ScalarParameter() {group=Face Flipbook, parameterName=Eyelids_Lower_R}) {constB=0.45}) {constB=0.5}
  n44 = Add(A=@n43) {constB=-0.04}
  n45 = LinearInterpolate(
   A=OneMinus(None=Ceil(None=@n40))
   B=LinearInterpolate(A=OneMinus(None=@n42.G), B=Floor(None=@n44), Alpha=Floor(None=@n43)) {constA=0.2}
   Alpha=Ceil(None=Subtract(A=@n40) {constB=0.04})
  ) {constA=0.2, constB=0.2}
  NamedRerouteDeclaration(None=LinearInterpolate(B=NamedRerouteUsage(), Alpha=LinearInterpolate(B=@n45, Alpha=@n31.R))) {name=Color_EyeR, nodeColor=<rgba 0.0547 0 1 1>}
 ]}
