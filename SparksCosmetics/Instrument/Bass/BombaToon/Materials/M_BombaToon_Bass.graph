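# Cloth-shaded anime toon material for the BombaToon bass instrument cosmetic on skeletal meshes, built on the shared 2023 Anime Parent functions.
# LegCircleL and LegCircleR parameters shift V inside a hard-masked UV rectangle, split along V into two halves.
# These custom UVs feed the Pre function, then Calculations derives shading gradients.
# MF_BombaToon_Bass adjusts the colors before the Post function combines everything.
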Material /SparksCosmetics/Instrument/Bass/BombaToon/Materials/M_BombaToon_Bass
graph {
 properties {
  bAutomaticallySetUsageInEditor=false
  bUseMaterialAttributes=true
  bUsedWithMeshDeformer=true
  bUsedWithSkeletalMesh=true
  pixelDepthOffsetMode=PDOM_Legacy
  shadingModel=MSM_Cloth
  usedShadingModels=Cloth
 }
 nodes [
  n1 = Constant()
  n2 = ScalarParameter() {group=Face Flipbook, parameterName=LegCircleL}
  n3 = TextureCoordinate()
  n4 = Add(A=AppendVector(A=@n1, B=Divide(A=@n2) {constB=100}), B=@n3) {constB=0.2}
  n5 = ScalarParameter() {group=Face Flipbook, parameterName=LegCircleR}
  n6 = Add(A=@n3, B=AppendVector(A=@n1, B=Divide(A=@n5) {constB=100})) {constB=0.2}
  n7 = TextureCoordinate()
  n8 = ComponentMask(None=@n7) {g=true}
  n9 = Saturate(None=Multiply(A=Add(A=@n8) {constB=-0.8654}) {constB=100})
  n10 = Saturate(None=Multiply(A=Add(A=@n8) {constB=-0.7542}) {constB=100})
  n11 = Add(A=@n8) {constB=-0.963}
  n12 = OneMinus(None=Saturate(None=Multiply(A=@n11) {constB=100}))
  n13 = ComponentMask(None=@n7) {r=true}
  n14 = Add(A=@n13) {constB=-0.75}
  n15 = OneMinus(None=Saturate(None=Multiply(A=@n14) {constB=100}))
  n16 = Saturate(None=Multiply(A=Add(A=@n13) {constB=-0.67}) {constB=100})
  n17 = Multiply(A=Saturate(None=Multiply(A=@n10, B=@n12)), B=Saturate(None=Multiply(A=@n15, B=@n16)))
  n18 = MaterialFunctionCall!(CustomUV=LinearInterpolate(A=TextureCoordinate(), B=LinearInterpolate(A=@n4, B=@n6, Alpha=@n9), Alpha=Saturate(None=@n17))) {materialFunction=/Game/Global/Materials/Character_Parent_2023/Anime_Parent_2023/MaterialFunctions/MFC_AnimeShader_2023_Pre}
  n19 = MaterialFunctionCall!(CustomLightVector, OverrideNormal_WS, STT=@n18.STT, Normal_Map=@n18.STT, DFL=@n18.STT, FX=@n18.STT) {materialFunction=/Game/Global/Materials/Character_Parent_2023/Anime_Parent_2023/MaterialFunctions/MFC_AnimeShader_2023_Calculations}
  n20 = MaterialFunctionCall!(Color=@n18.Color_Lit, Shade=@n18.Color_Lit) {materialFunction=&/MF_BombaToon_Bass}
  MaterialFunctionCall!(
   Shading_Gradient=@n19.Shading_Gradient
   Threshold_Gradient=@n19.Shading_Gradient
   Specular_Gradient=@n19.Shading_Gradient
   Sheen_Mask=@n19.Shading_Gradient
   Thickness=@n19.Shading_Gradient
   DFL_Mask=@n19.Shading_Gradient
   FX=@n19.Shading_Gradient
   Metallic_Gradient=@n19.Shading_Gradient
   RimLight_Gradient=@n19.Shading_Gradient
   NormalBuffer=@n19.Shading_Gradient
   --------
   CINE_LightSources=@n19.Shading_Gradient
   CINE_LightColor=@n19.Shading_Gradient
   ---------
   Color_Lit=@n20.Result_Color
   Color_Shaded=@n20.Result_Color
   SRM=@n18.SRM
   Emissive=@n18.SRM
   DFL_Color=@n18.SRM
   Specular_Color=@n18.SRM
   RimLight_Color=@n18.SRM
  ) {materialFunction=/Game/Global/Materials/Character_Parent_2023/Anime_Parent_2023/MaterialFunctions/MFC_AnimeShader_2023_Post}
 ]
}
