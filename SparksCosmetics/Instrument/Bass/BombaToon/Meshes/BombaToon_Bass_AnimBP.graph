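# Animation Blueprint for the BombaToon bass cosmetic, animating its character legs over the input pose.
# An integer state selects idle, happy, surprised or hide poses from player events such as swings, chests, eliminations and health or shield changes.
# Idle alternates with two beat-synced loops when a music clock is valid, with curve-weighted AnimDynamics on legs and shins.
# Previews cycle reactions in shuffled order, and frontend display overrides material LUT settings.
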AnimBlueprint /SparksCosmetics/Instrument/Bass/BombaToon/Meshes/BombaToon_Bass_AnimBP
target_skeleton=&/BombaToon_Bass_Skeleton
graphs [
 AnimGraph {nodes [
   Modify Curve!(
    SourcePose
    CurveValues_0=Get FrozenMouthIndex() -> (FrozenMouthIndex:float)
    CurveValues_1:float=Get Frozen Mouth Vertical Flip() -> (Frozen Mouth Vertical Flip:float)
    CurveValues_2:float=Get Mouth Scale V() -> (Mouth Scale V:float)
    CurveValues_3:float=Get Mouth Scale U() -> (Mouth Scale U:float)
    CurveValues_4:float=Get Mouth Position V() -> (Mouth Position V:float)
    CurveValues_5:float=Get Mouth Position U() -> (Mouth Position U:float)
    Alpha=Get OverrideMouthIndex() -> (OverrideMouthIndex:float)
   ) -> (Pose) {curveNames [mouth_index, Mouth_VerticalFlip, mouth_scale_v, mouth_scale_u, mouth_position_v, mouth_position_u], curveValues [0,0,0,0,0,0]}
   n1 = Blend Poses by bool(
    BlendPose_0=Blend Poses by bool(
     BlendPose_0=Play Beat Synced Animation!(
      ExplicitTime:float=0
      bShouldLoop:bool=True
      InSequence:Anim Sequence Base Object Reference=/SparksCosmetics/Instrument/Bass/BombaToon/Animations/BombaToon_Idle_Beatsynced
      MusicClock:Music Clock Object Reference
      TempoMultiplier:float=1
      bEnableTempoVariants:bool=False
      TempoVariantHysteresis:float=0.05
      bShowDebug:bool=False
     ) -> (Pose:Pose Link Structure) {is=title AnimGraphNode_PlayBeatSyncedAnim}
     BlendPose_1=Use cached pose 'BeatSync_1'()
     BlendTime_0=0
     BlendTime_1=0
     bActiveValue=Get ClockValid() -> (ClockValid:bool)
    )
    BlendPose_1=Blend Poses by bool(
     BlendPose_0=Play Beat Synced Animation(
      ExplicitTime=0
      bShouldLoop=True
      InSequence=/SparksCosmetics/Instrument/Bass/BombaToon/Animations/BombaToon_LegSwing
      TempoMultiplier=1
      bEnableTempoVariants=False
      TempoVariantHysteresis=0.05
      bShowDebug=False
     )
     BlendPose_1=Use cached pose 'BeatSync_2'()
     BlendTime_0=0
     BlendTime_1=0
     bActiveValue=Get ClockValid()
    )
    BlendTime_0=0.2
    BlendTime_1=0.2
    bActiveValue=Get AltBeatSyncedAnim() -> (AltBeatSyncedAnim:bool)
   )
   n2 = Use cached pose 'Idle'()
   n3 = Get PlayBeatSynced() -> (PlayBeatSynced:bool)
   n4 = Get CustomAnimIsPlaying() -> (CustomAnimIsPlaying:double)
   n5 = Anim Dynamics - Bone: leg_l!(
    ComponentPose=Local To Component(LocalPose=Blend Poses by bool(BlendPose_0=@n1, BlendPose_1=@n2, bActiveValue=@n3))
    Alpha=@n4
    GravityOverride:Vector="0,0,0"
    ExternalForce="0,0,0"
   ) -> (Pose) {
    alphaCurveName=DynamicsAlpha
    angularDampingOverride=1
    angularSpringConstant=3500
    bAngularSpring=true
    bOverrideAngularDamping=true
    bUseGravityOverride=true
    gravityScale=0.35
    simulationSpace=RootRelative
    boundBone {boneName=leg_l}
    componentAppliedLinearAccClamp=<2400 2400 2400>
    componentLinearAccScale=<0.24 0.24 0.24>
    componentLinearVelScale=<0.24 0.24 0.24>
    gravityOverride {z=-250}
    physicsBodyDefinitions [{
      boundBone {boneName=leg_l}
      boxExtents=<25 20 25>
      constraintSetup {angularLimitsMax=<0 25 40>, angularLimitsMin=<0 -25 -40>, angularTarget=<1 0 0>}
      localJointOffset=<3 0 0>
     }]
    relativeSpaceBone {boneName=leg_l}
    simSpaceSettings {maxAngularAcceleration=120, maxAngularVelocity=80, simSpaceAngularAlpha=0.5}
   }
   n6 = Anim Dynamics - Bone: leg_r!(ComponentPose=@n5, Alpha=@n4, GravityOverride:Vector="0,0,-300", ExternalForce="0,0,0") -> (Pose) {
    angularDampingOverride=1
    angularSpringConstant=3500
    bAngularSpring=true
    bOverrideAngularDamping=true
    bUseGravityOverride=true
    gravityScale=0.35
    simulationSpace=RootRelative
    boundBone {boneName=leg_r}
    componentAppliedLinearAccClamp=<2400 2400 2400>
    componentLinearAccScale=<0.24 0.24 0.24>
    componentLinearVelScale=<0.24 0.24 0.24>
    gravityOverride {z=-300}
    physicsBodyDefinitions [{
      boundBone {boneName=leg_r}
      boxExtents=<24 19 24>
      constraintSetup {angularLimitsMax=<0 25 45>, angularLimitsMin=<0 -25 -40>, angularTarget=<1 0 0>}
      localJointOffset=<-2.5 0 0>
     }]
    relativeSpaceBone {boneName=leg_l}
    simSpaceSettings {maxAngularAcceleration=120, maxAngularVelocity=80, simSpaceAngularAlpha=0.4}
   }
   n7 = Anim Dynamics - Bone: shin_l!(ComponentPose=@n6, Alpha=@n4, GravityOverride:Vector="0,0,0", ExternalForce="0,0,0") -> (Pose) {
    angularDampingOverride=1
    angularSpringConstant=3500
    bAngularSpring=true
    bOverrideAngularDamping=true
    bUseGravityOverride=true
    gravityScale=0.35
    simulationSpace=RootRelative
    boundBone {boneName=shin_l}
    componentAppliedLinearAccClamp=<2400 2400 2400>
    componentLinearAccScale=<0.24 0.24 0.24>
    componentLinearVelScale=<0.24 0.24 0.24>
    gravityOverride {z=-250}
    physicsBodyDefinitions [{
      boundBone {boneName=shin_l}
      boxExtents=<20 15 20>
      constraintSetup {angularLimitsMax=<0 3 10>, angularLimitsMin=<0 -3 -15>, angularTarget=<1 0 0>}
      localJointOffset=<3 0 0>
     }]
    relativeSpaceBone {boneName=leg_l}
    simSpaceSettings {maxAngularAcceleration=120, maxAngularVelocity=80, simSpaceAngularAlpha=0.5}
   }
   n8 = Anim Dynamics - Bone: shin_r!(ComponentPose=@n7, Alpha=@n4, GravityOverride:Vector="0,0,-300", ExternalForce="0,0,0") -> (Pose) {
    angularDampingOverride=1
    angularSpringConstant=3100
    bAngularSpring=true
    bOverrideAngularDamping=true
    bUseGravityOverride=true
    gravityScale=0.35
    simulationSpace=RootRelative
    boundBone {boneName=shin_r}
    componentAppliedLinearAccClamp=<2400 2400 2400>
    componentLinearAccScale=<0.24 0.24 0.24>
    componentLinearVelScale=<0.24 0.24 0.24>
    gravityOverride {z=-300}
    physicsBodyDefinitions [{
      boundBone {boneName=shin_r}
      boxExtents=<20 15 20>
      constraintSetup {angularLimitsMax=<0 3 10>, angularLimitsMin=<0 -3 -15>, angularTarget=<1 0 0>}
      localJointOffset=<-2.5 0 0>
     }]
    relativeSpaceBone {boneName=leg_l}
    simSpaceSettings {maxAngularAcceleration=120, maxAngularVelocity=80, simSpaceAngularAlpha=0.4}
   }
   n9 = Blend Poses by int!(
    BlendPose_0:Pose Link Structure=Component To Local(ComponentPose=@n8)
    BlendPose_1:Pose Link Structure=Use cached pose 'Happy'()
    BlendPose_2:Pose Link Structure=Use cached pose 'Surprised'()
    BlendPose_3:Pose Link Structure=Use cached pose 'Hide'()
    BlendTime_0:float=0.15
    BlendTime_1:float=0.1
    BlendTime_2:float=0.1
    BlendTime_3:float=0.1
    ActiveChildIndex:int=Get State() -> (State:int=0)
   ) -> (Pose:Pose Link Structure) {
    is=title AnimGraphNode_BlendListByInt
    blendPose [{linkId=-1, sourceLinkId=-1}, {linkId=-1, sourceLinkId=-1}, {linkId=-1, sourceLinkId=-1}, {linkId=-1, sourceLinkId=-1}]
    blendTime [0.1,0.1,0.1,0.1]
   }
   n10 = Blend Poses by bool(
    BlendPose_0=Play Beat Synced Animation(
     ExplicitTime=0
     bShouldLoop=True
     InSequence=&BombaToon_LegSwing
     TempoMultiplier=1
     bEnableTempoVariants=False
     TempoVariantHysteresis=0.05
     bShowDebug=False
    )
    BlendPose_1=Use cached pose 'BeatSync_2'()
    BlendTime_0=0
    BlendTime_1=0
    bActiveValue=Get ClockValid()
   )
   n11 = Blend Poses by bool(
    BlendPose_0=Play Beat Synced Animation(
     ExplicitTime=0
     bShouldLoop=True
     InSequence=&BombaToon_Idle_Beatsynced
     TempoMultiplier=1
     bEnableTempoVariants=False
     TempoVariantHysteresis=0.05
     bShowDebug=False
    )
    BlendPose_1=Use cached pose 'BeatSync_1'()
    BlendTime_0=0
    BlendTime_1=0
    bActiveValue=Get ClockValid()
   )
   n12 = Get AltBeatSyncedAnim()
   n13 = Blend Poses by int(
    BlendPose_0=Blend Poses by bool(
     BlendPose_0=Blend Poses by bool(BlendPose_0=@n11, BlendPose_1=@n10, BlendTime_0=0, BlendTime_1=0, bActiveValue=@n12)
     BlendPose_1=Use cached pose 'Idle'()
     BlendTime_0=0
     BlendTime_1=0
     bActiveValue=Get PlayBeatSynced()
    )
    BlendPose_1=Use cached pose 'Happy'()
    BlendPose_2=Use cached pose 'Surprised'()
    BlendPose_3=Use cached pose 'Hide'()
    BlendTime_0=0
    BlendTime_1=0
    BlendTime_2=0
    BlendTime_3=0
    ActiveChildIndex=Get State() -> (State=0)
   ) {
    blendPose [{linkId=-1, sourceLinkId=-1}, {linkId=-1, sourceLinkId=-1}, {linkId=-1, sourceLinkId=-1}, {linkId=-1, sourceLinkId=-1}]
    blendTime [0.1,0.1,0.1,0.1]
   }
   n14 = Layered blend per bone(BasePose=@n13, BlendPoses_0=@n9) {curveBlendOption=UseBasePose, layerSetup [{branchFilters [{boneName=leg_l}, {boneName=leg_r}]}]}
   n15 = Layered blend per bone(BasePose=Input Pose(), BlendPoses_0=@n14) {layerSetup [{branchFilters [
       {boneName=leg_l}
       {boneName=shin_l}
       {boneName=foot_l}
       {boneName=toes_l}
       {boneName=leg_r}
       {boneName=shin_r}
       {boneName=foot_r}
       {boneName=toes_r}
      ]}]}
   Output Pose(Result=@n15)
   Save cached pose 'BeatSync_2'(Pose=Sequence Player 'BombaToon_LegSwing'!(Sequence:Anim Sequence Base Object Reference=&BombaToon_LegSwing) -> (Pose)) {cacheName=BeatSync_2}
   Save cached pose 'Idle'(Pose=Sequence Player 'BombaToon_Idle'() {sequence=/SparksCosmetics/Instrument/Bass/BombaToon/Animations/BombaToon_Idle}) {cacheName=Idle}
   AnimGraphNode_StateMachine_0 = SurprisedNoLoop() {is=title AnimGraphNode_StateMachine}
   Save cached pose 'Surprised'(Pose=@AnimGraphNode_StateMachine_0) {cacheName=Surprised}
   Save cached pose 'Happy'(Pose=Sequence Player 'BombaToon_LegHappy'() {sequence=/SparksCosmetics/Instrument/Bass/BombaToon/Animations/BombaToon_LegHappy}) {cacheName=Happy}
   Save cached pose 'Hide'(Pose=Sequence Player 'BombaToon_LegHide'() {sequence=/SparksCosmetics/Instrument/Bass/BombaToon/Animations/BombaToon_LegHide}) {cacheName=Hide}
   Save cached pose 'BeatSync_1'(Pose=Sequence Player 'BombaToon_Idle_Beatsynced'!(Sequence:Anim Sequence Base Object Reference=&BombaToon_Idle_Beatsynced) -> (Pose)) {cacheName=BeatSync_1}
   Use cached pose 'Surprised'()
  ]}
 SurprisedNoLoop {entry=State, path=AnimGraph.AnimGraphNode_StateMachine_0.SurprisedNoLoop, ids=id, nodes [AnimStateNode_0 = State() {is=title AnimStateNode}]}
 State {
  path=AnimGraph.AnimGraphNode_StateMachine_0.SurprisedNoLoop.AnimStateNode_0.State
  nodes [Output Animation Pose(Result=Sequence Player 'BombaToon_LegSurprised'() {sequence=/SparksCosmetics/Instrument/Bass/BombaToon/Animations/BombaToon_LegSurprised})]
 }
 EventGraph {nodes [
   n1 = SwitchStateAndFreezeMouth!() -> (OutputDelegate, then, State:int, CheckState:bool=true) {is=title K2Node_CustomEvent}
   > n2 = Branch(Condition=@n1.CheckState)
   > n3 = Branch(Condition="Not Equal (Integer)"(A=@n1.State, B=Get State() -> (State=0)))
   n4 = Get Dynamic Material Instance(execute=[@n12.then, @n3.then], SkeletalMeshComponent=GetOwningComponent())
   > n5 = Sequence!(execute) -> (then_0, then_1, then_2:Exec, then_3:Exec, then_4:Exec)
   > n6 = Get Scalar Parameter Value!(execute, self:Material Instance Dynamic Object Reference=@n4, ParameterName:Name=Mouth_Index) -> (then, ReturnValue:float) {is=title K2Node_CallFunction}
   > Set FrozenMouthIndex(FrozenMouthIndex:float=@n6) -> (Output_Get:float)
   n7 = Get Scalar Parameter Value(execute=@n5.then_1, self=@n4, ParameterName=Mouth_VerticalFlip)
   > Set Frozen Mouth Vertical Flip(Frozen Mouth Vertical Flip:float=@n7) -> (Output_Get:float)
   n8 = Get Scalar Parameter Value(execute=@n5.then_2, self=@n4, ParameterName=Mouth_Scale_V)
   > Set Mouth Scale V(Mouth Scale V:float=@n8) -> (Output_Get:float)
   > n9 = Get Scalar Parameter Value(self=@n4, ParameterName=Mouth_Scale_U)
   > Set Mouth Scale U(Mouth Scale U:float=@n9) -> (Output_Get:float)
   n10 = Get Scalar Parameter Value(execute=@n5.then_3, self=@n4, ParameterName=Mouth_Position_V)
   > Set Mouth Position V(Mouth Position V:float=@n10) -> (Output_Get:float)
   > n11 = Get Scalar Parameter Value(self=@n4, ParameterName=Mouth_Position_U)
   > Set Mouth Position U(Mouth Position U:float=@n11) -> (Output_Get:float)
   Set OverrideMouthIndex(execute=@n5.then_4, OverrideMouthIndex:float=1) -> (Output_Get:float)
   > Set State(State:int=@n1.State) -> (Output_Get:int=0)
   > RetriggerableDelay(Duration=0.12)
   > Set OverrideMouthIndex(OverrideMouthIndex=0)
   n12 = Branch(execute=@n2.else, Condition="Equal (Integer)"(A=Get State() -> (State=0)))
   StartSwinging() {is=title K2Node_CustomEvent}
   n13 = Get State() -> (State=0)
   > Branch(Condition=OR Boolean(A="Equal (Integer)"(A=@n13), B="Equal (Integer)"(A=@n13, B=2)))
   > n14 = Set State(State=2) -> (Output_Get=0)
   RetriggerableDelay(
    execute=[@n48.then, @n49.then, @n14.then]
    Duration=Select!(Option 0:float=0, Option 1:float=3, Option 2:float=0.7, Option 3:float=2, Index:int=Get State() -> (State=0)) -> (ReturnValue:float)
   )
   > Set State(State=0) -> (Output_Get=0)
   n15 = GetClock!() -> (OutputDelegate, then, DeltaTime:double=~) {is=title K2Node_CustomEvent}
   n16 = GetHighestPriorityClock!(self:Music Event Subsystem Object Reference=Get MusicEventSubsystem() -> (ReturnValue:Music Event Subsystem Object Reference) {is=title K2Node_GetSubsystem}) -> (ReturnValue:Music Clock Object Reference)
   > n17 = Branch(Condition=IsValid(Object=@n16))
   > Set Clock(Clock:Music Clock Object Reference=@n16) -> (Output_Get:Music Clock Object Reference)
   > n18 = Set ClockValid(ClockValid:bool="Not Equal (Float)"(A=GetCurrentTempo!(self:Music Clock Object Reference=@n16) -> (ReturnValue:float))) -> (Output_Get:bool)
   > n19 = Branch(Condition=@n18)
   > Set PlayBeatSynced(PlayBeatSynced:bool=@n18) -> (Output_Get:bool)
   > Set BeatsyncedTime(BeatsyncedTime:double=float - float(A=Get BeatsyncedTime() -> (BeatsyncedTime:double), B=@n15.DeltaTime)) -> (Output_Get:double)
   n20 = Branch(execute=[@n19.else, @n21.then], Condition=Get IsPreview() -> (IsPreview:bool))
   Set PlayBeatSynced(execute=@n20.else, PlayBeatSynced=false)
   Set Clock(execute=@n17.else)
   > n21 = Set ClockValid(ClockValid=false)
   Event BlueprintBeginPlay() {is=title K2Node_Event}
   > n22 = Sequence!(execute) -> (then_0, then_1, then_2)
   > n23 = Cast To PlayerPawn_Athena(Object=TryGetPawnOwner())
   n24 = Create Event()
   > Bind Event to Swing Left(self:Player Pawn Athena Object Reference=@n23, Delegate=@n24) {is=title K2Node_AddDelegate}
   > Bind Event to Swing Right(self:Player Pawn Athena Object Reference=@n23, Delegate=@n24) {is=title K2Node_AddDelegate}
   n25 = GetOwningActor()
   n26 = Set IsPreview(
    execute=@n22.then_1
    IsPreview:bool=OR Boolean(
     A=IsCosmeticObjectDisplayedAsPreview!(CosmeticObject:Object Reference=@n25) -> (ReturnValue:bool)
     B=Cast To B_SparksCosmeticPreviewDisplay!(Object=@n25) -> (AsB Sparks Cosmetic Preview Display:B Sparks Cosmetic Preview Display Object Reference, bSuccess:bool=~) {is=title K2Node_DynamicCast}.bSuccess
    )
   ) -> (Output_Get:bool)
   > Branch(Condition=@n26)
   > Do Once()
   > Shuffle!(execute, TargetArray:Array of Integers=Get RandomSequence() -> (RandomSequence:Array of Integers)) -> (then) {is=title K2Node_CallArrayFunction}
   > n27 = Delay(Duration=8.33)
   n28 = MultiGate!(execute:Exec=[@n27.then, @n34.then], Reset:Exec, IsRandom:bool, Loop:bool=true, StartIndex:int=-1) -> (Out 0:Exec, Out 1:Exec, Out 2:Exec, Out 3:Exec, Out 4:Exec) {is=title K2Node_MultiGate}
   n29 = Get RandomSequence()
   > n30 = Set RandomIntSequence(RandomIntSequence:int="Get (a copy)"(Array:Array of Integers=@n29) -> (Output:int)) -> (Output_Get:int=0)
   n31 = Switch on Int!(execute:Exec=[@n41.then, @n42.then, @n43.then, @n30.then, @n40.then], Selection:int=Get RandomIntSequence() -> (RandomIntSequence:int=0)) -> (Default:Exec, 0:Exec, 1:Exec, 2:Exec, 3:Exec, 4:Exec)
   n32 = Set State(execute=@n31.0, State=1) -> (Output_Get=0)
   Delay(execute=[@n35.then, @n36.then, @n32.then], Duration=3)
   > n33 = Set State(State=0) -> (Output_Get=0)
   n34 = Delay(execute=[@n38.then, @n33.then], Duration=8.33)
   n35 = Set State(execute=@n31.1, State=2) -> (Output_Get=0)
   n36 = Set State(execute=@n31.2, State=3) -> (Output_Get=0)
   Set PlayBeatSynced(execute=@n31.3, PlayBeatSynced=true)
   > n37 = Set AltBeatSyncedAnim(AltBeatSyncedAnim:bool=false) -> (Output_Get:bool)
   Delay(execute=[@n37.then, @n39.then], Duration=6)
   > Set PlayBeatSynced(PlayBeatSynced=false)
   > n38 = Set AltBeatSyncedAnim(AltBeatSyncedAnim=false)
   Set PlayBeatSynced(execute=@n31.4, PlayBeatSynced=true)
   > n39 = Set AltBeatSyncedAnim(AltBeatSyncedAnim=true)
   n40 = Set RandomIntSequence(execute=@n28.Out 1, RandomIntSequence="Get (a copy)"(Array=@n29, Dimension 1=1)) -> (Output_Get=0)
   n41 = Set RandomIntSequence(execute=@n28.Out 2, RandomIntSequence="Get (a copy)"(Array=@n29, Dimension 1=2)) -> (Output_Get=0)
   n42 = Set RandomIntSequence(execute=@n28.Out 3, RandomIntSequence="Get (a copy)"(Array=@n29, Dimension 1=3)) -> (Output_Get=0)
   n43 = Set RandomIntSequence(execute=@n28.Out 4, RandomIntSequence="Get (a copy)"(Array=@n29, Dimension 1=4)) -> (Output_Get=0)
   DelayUntilNextTick!(execute=@n22.then_2) -> (then)
   > Branch(Condition=IsCosmeticObjectDisplayedInFrontendContext!(CosmeticObject:Object Reference=GetOwningActor()) -> (ReturnValue:bool))
   > n44 = Get Dynamic Material Instance(SkeletalMeshComponent=GetOwningComponent())
   > SetTextureParameterValue(self=[@n44, @n44], ParameterName=LUT Volume Texture, Value=/Game/Packages/Fortress_SharedMaterials/Anime_Master_Material/VT_InverseLUT_FE)
   > SetScalarParameterValue(self=@n44, ParameterName=LUT Multiplier, Value=1.35)
   > SetScalarParameterValue(self=@n44, ParameterName=Mobile LUT Brightness, Value=1)
   > SetScalarParameterValue(self=@n44, ParameterName=use 33 Resolution, Value=0)
   > SetScalarParameterValue(self=@n44, ParameterName=use Color Adjustment Instead of LUT, Value=0)
   n45 = Event BlueprintProcessAnimInputEvent()
   > n46 = Branch(Condition=Get IsPreview())
   n47 = Switch on Name!(
    execute:Exec=@n46.else
    Selection:Name=Get EventName!(self:Fort Anim Input Event Object Reference=@n45.AnimInputEvent) -> (EventName:Name=None)
   ) -> (
    aie.EnemyPlayerEliminated:Exec
    aie.ChestOpened:Exec
    aie.PlayerVitality.ShieldsIncreased:Exec
    aie.PlayerVitality.HealthIncreased:Exec
    aie.PlayerVitality.ShieldsDecreased:Exec
    aie.PlayerVitality.HealthDecreased:Exec
   ) {
    is=title K2Node_SwitchName
    bHasDefaultPin=false
    pinNames [
     aie.EnemyPlayerEliminated
     aie.ChestOpened
     aie.PlayerVitality.ShieldsIncreased
     aie.PlayerVitality.HealthIncreased
     aie.PlayerVitality.ShieldsDecreased
     aie.PlayerVitality.HealthDecreased
    ]
   }
   n48 = Set State(
    execute=[@n47.aie.ChestOpened, @n47.aie.EnemyPlayerEliminated, @n47.aie.PlayerVitality.HealthIncreased, @n47.aie.PlayerVitality.ShieldsIncreased]
    State=1
   ) -> (Output_Get=0)
   n49 = Set State(execute=[@n47.aie.PlayerVitality.HealthDecreased, @n47.aie.PlayerVitality.ShieldsDecreased], State=3) -> (Output_Get=0)
   AnimNotify_SwapIdle() {is=title K2Node_Event}
   > n50 = Branch(Condition=Get IsPreview())
   n51 = GetGameTimeInSeconds()
   Branch(execute=@n50.else, Condition="float < float"(A=Get BeatsyncedTime(), B=@n51))
   > Set BeatsyncedTime(BeatsyncedTime=float + float(A=@n51, B=RandomFloatInRange(Min=10, Max=16)))
   > n52 = Flip Flop!(None:Exec) -> (A:Exec, B:Exec, IsA:bool=~) {is=title K2Node_MacroInstance}
   Set AltBeatSyncedAnim(execute=[@n52.A, @n52.B], AltBeatSyncedAnim=@n52)
   n53 = Event BlueprintUpdateAnimation()
   > GetClock!(execute, self, DeltaTime:double=@n53.DeltaTimeX) -> (then) {is=title K2Node_CallFunction}
   > Set CustomAnimIsPlaying(CustomAnimIsPlaying:double=FInterpTo(
     Current=Get CustomAnimIsPlaying()
     Target=GetCurveValueWithDefault!(self:Anim Instance Object Reference, CurveName:Name=DynamicsAlpha, DefaultValue:float=0) -> (OutValue:float, ReturnValue:bool)
     DeltaTime=@n53.DeltaTimeX
     InterpSpeed=27
    )) -> (Output_Get:double)
   > n54 = Branch(Condition=Get bIsSkydiveFloating())
   > Set DynGravityLeftLeg(DynGravityLeftLeg:Vector="0,0,250") -> (Output_Get:Vector)
   > Set DynGravityRightLeg(DynGravityRightLeg:Vector="0,0,300") -> (Output_Get:Vector)
   Set DynGravityLeftLeg(execute=@n54.else, DynGravityLeftLeg="0,0,-250")
   > Set DynGravityRightLeg(DynGravityRightLeg="0,0,-300")
   SwitchStateAndFreezeMouth!(execute, self, State:int=0, CheckState:bool=false) -> (then) {is=title K2Node_CallFunction}
  ]}
]
variables [
 AltBeatSyncedAnim {type=boolean}
 BeatsyncedTime {type=number}
 BeatSyncSequence {type=AnimSequenceBase, default=&BombaToon_Idle_Beatsynced}
 Clock {type=MusicClockComponent, default=null}
 ClockValid {type=boolean}
 CustomAnimIsPlaying {type=number}
 DynGravityLeftLeg {type=Vector, default=<0 0 -250>}
 DynGravityRightLeg {type=Vector, default=<0 0 -300>}
 Frozen Mouth Vertical Flip {type=number}
 FrozenMouthIndex {type=number}
 IsPreview {type=boolean}
 Mouth Position U {type=number}
 Mouth Position V {type=number}
 Mouth Scale U {type=number}
 Mouth Scale V {type=number}
 OverrideMouthIndex {type=number}
 PlayBeatSynced {type=boolean}
 RandomIntSequence {type=integer}
 RandomSequence {type="array<integer>", default [0,1,2,3,4]}
 State {type=integer}
]
